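; Looping gameplay cue for entering a fire pit hazard vent, adding first-person activation audio for the affected player.
; On looping start it calls PlayActivationAudio on the cue target, then runs the parent implementation.
; The function casts the target to a player pawn and plays sound only if that pawn is locally viewed.
; Pawns in vehicles get the vehicle cue attached to the vehicle's skeletal mesh; others get the player cue on their own mesh.
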(event EventOnLoopingStart(EitherVFX-preferthisversion)
  (params (MyTarget :type Actor%20Object%20Reference) (Parameters :type Gameplay%20Cue%20Parameters%20Structure%20%28by%20ref%29) (ParticleComponents :type Array%20of%20FXSystem%20Component%20Object%20References) (AudioComponents :type Array%20of%20Audio%20Component%20Object%20References))
  (CallFunction|PlayActivationAudio MyTarget)
  (|CallParentFunction MyTarget Parameters ParticleComponents AudioComponents))

(fn PlayActivationAudio
  (params (MyTarget :type Object%20Reference))
  (bind _playerpawn (LocalVariables|GetPlayerPawn))
  (bind _asplayer_pawn_athena (Utilities|Casting|CastToPlayerPawn_Athena MyTarget))
  (bind _output_get (LocalVariables|SetPlayerPawn _asplayer_pawn_athena))
  (if (Pawn|IsLocallyViewed _output_get)
    (if (Vehicle|IsinVehicle _output_get)
      (Audio|SpawnSoundAttached "/FirePitHazard/Assets/Audio/FirePitHazard_Activate_Vehicle1P_Cue.FirePitHazard_Activate_Vehicle1P_Cue" (Variables|/Script/FortniteGame.FortAthenaVehicle|GetSkeletalMesh (Vehicle|GetVehicle _playerpawn)))
      (else
        (Audio|SpawnSoundAttached "/FirePitHazard/Assets/Audio/FirePitHazard_Activate_Player1P_Cue.FirePitHazard_Activate_Player1P_Cue" (Variables|/Script/Engine.Character|GetMesh _playerpawn))))))

(fn ConstructionScript)
