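; Gameplay cue that plays a launch sound for the fire pit hazard when its burst fires.
; ChooseSoundAsset picks a first- or third-person sound through GetAudioforPlayerEvent on the instigator, and that result decides between the instigator's AudioComponent and the target's root as attach point.
; Without a valid instigator the launch counts as ambient, using the ambient sound on the target's AudioComponent.
; OnBurst spawns the chosen sound attached to the chosen component.
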(fn ChooseSoundAsset
  (params (TargetActor :type Actor%20Object%20Reference) (Instigator :type Actor%20Object%20Reference))
  (Utilities|IsValid Instigator
    (:Is%20Valid
      (return (AudioGameplay|GetAudioforPlayerEvent Instigator (Variables|Default|Get1P%20Sound) (Variables|Default|Get3P%20Sound)) (select (== (AudioGameplay|GetAudioforPlayerEvent Instigator (Variables|Default|Get1P%20Sound) (Variables|Default|Get3P%20Sound)) (Variables|Default|Get3P%20Sound)) (Actor|GetComponentbyClass Instigator "/Script/Engine.AudioComponent") (Variables|/Script/Engine.Actor|GetRootComponent TargetActor))))
    (:Is%20Not%20Valid
      (return (Variables|Default|GetAmbient%20Sound) (Actor|GetComponentbyClass TargetActor "/Script/Engine.AudioComponent")))) ; If Instigator is None, then this is an ambient launch
)

(fn OnBurst(EitherVFX-preferthisversion)
  (params (MyTarget :type Actor%20Object%20Reference) (Parameters :type Gameplay%20Cue%20Parameters%20Structure%20%28by%20ref%29) (ParticleComponents :type Array%20of%20FXSystem%20Component%20Object%20References) (AudioComponents :type Array%20of%20Audio%20Component%20Object%20References) (BurstCameraShakeInstance :type Camera%20Shake%20Base%20Object%20Reference) (BurstDecalInstance :type Decal%20Actor%20Object%20Reference))
  (bind (_normalizedmagnitude _rawmagnitude _effectcontext _matchedtagname _originaltag _aggregatedsourcetags _aggregatedtargettags _location _normal _instigator) (Ability|GameplayCue|BreakGameplayCueParameters Parameters))
  (bind (_audioasset _rootcomponent) (CallFunction|ChooseSoundAsset MyTarget _instigator))
  (Audio|SpawnSoundAttached _audioasset _rootcomponent)
  (return))
